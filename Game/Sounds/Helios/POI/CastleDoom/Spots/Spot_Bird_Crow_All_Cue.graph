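# Castle Doom ambient one-shot cue that plays a crow call drawn from mountain, boreal forest, corn maze and swamp recordings.
# All quality levels share one equally weighted random node that picks one of the four groups, each with its own random wave pool.
# Each pool runs through a modulator and a single-input mixer that sets group gain.
# Bird attenuation and the outdoor one-shot sound class apply, with a reverb submix send.
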SoundCue /Game/Sounds/Helios/POI/CastleDoom/Spots/Spot_Bird_Crow_All_Cue
root=QualityLevel_0
nodes [
 {id=Mixer_0, children [Modulator_0], properties {inputVolume [0.4]}}
 {id=Mixer_1, children [Modulator_1], properties {inputVolume [0.4]}}
 {id=Mixer_2, children [Modulator_2], properties {inputVolume [0.25]}}
 {id=Mixer_3, children [Modulator_3], properties {inputVolume [1]}}
 {id=Modulator_0, children [Random_1], properties {pitchMin=0.8, volumeMax=1, volumeMin=0.4}}
 {id=Modulator_1, children [Random_2]}
 {id=Modulator_2, children [Random_3]}
 {id=Modulator_3, children [Random_4]}
 {id=QualityLevel_0, children [Random_5,Random_5,Random_5,Random_5]}
 {id=Random_1, children [WavePlayer_3,WavePlayer_4,WavePlayer_5], properties {weights [1,1,1]}}
 {id=Random_2, children [WavePlayer_2, WavePlayer_0, WavePlayer_9, WavePlayer_11], properties {weights [1,1,1,1]}}
 {
  id=Random_3
  children [WavePlayer_12, WavePlayer_13, WavePlayer_14, WavePlayer_15, WavePlayer_16, WavePlayer_17]
  properties {weights [1,1,1,1,1,1]}
 }
 {id=Random_4, children [WavePlayer_18, WavePlayer_19, WavePlayer_20, WavePlayer_21, WavePlayer_22], properties {weights [1,1,1,1,1]}}
 {id=Random_5, children [Mixer_0,Mixer_1,Mixer_2,Mixer_3], properties {weights [1,1,1,1]}}
 {id=WavePlayer_0, properties {soundWaveAssetPtr=/Game/Sounds/Helios/Biomes/BorealForest/Spots/Spot_Bird_Crow_DayEvening_Eastern_Europe_03}}
 {id=WavePlayer_11, properties {soundWaveAssetPtr=/Game/Sounds/Helios/Biomes/BorealForest/Spots/Spot_Bird_Crow_DayEvening_Eastern_Europe_01}}
 {id=WavePlayer_12, properties {soundWaveAssetPtr=/Game/Sounds/Artemis/POI/Cornmaze/Spots/Spot_CornMaze_Crow_02}}
 {id=WavePlayer_13, properties {soundWaveAssetPtr=/Game/Sounds/Artemis/POI/Cornmaze/Spots/Spot_CornMaze_Crow_05}}
 {id=WavePlayer_14, properties {soundWaveAssetPtr=/Game/Sounds/Artemis/POI/Cornmaze/Spots/Spot_CornMaze_Crow_01}}
 {id=WavePlayer_15, properties {soundWaveAssetPtr=/Game/Sounds/Artemis/POI/Cornmaze/Spots/Spot_CornMaze_Crow_03}}
 {id=WavePlayer_16, properties {soundWaveAssetPtr=/Game/Sounds/Artemis/POI/Cornmaze/Spots/Spot_CornMaze_Crow_04}}
 {id=WavePlayer_17, properties {soundWaveAssetPtr=/Game/Sounds/Artemis/POI/Cornmaze/Spots/Spot_CornMaze_Crow_06}}
 {id=WavePlayer_18, properties {soundWaveAssetPtr=&/Spot_Bird_Swamp_Crow_01}}
 {id=WavePlayer_19, properties {soundWaveAssetPtr=&/Spot_Bird_Swamp_Crow_02}}
 {id=WavePlayer_2, properties {soundWaveAssetPtr=/Game/Sounds/Helios/Biomes/BorealForest/Spots/Spot_Bird_Crow_DayEvening_Eastern_Europe_04}}
 {id=WavePlayer_20, properties {soundWaveAssetPtr=&/Spot_Bird_Swamp_Crow_03}}
 {id=WavePlayer_21, properties {soundWaveAssetPtr=&/Spot_Bird_Swamp_Crow_04}}
 {id=WavePlayer_22, properties {soundWaveAssetPtr=&/Spot_Bird_Swamp_Crow_05}}
 {id=WavePlayer_3, properties {soundWaveAssetPtr=/Game/Sounds/Asteria/Biomes/Eastern/Spots/Spot_Bird_Rare_Mountain_Crow_01}}
 {id=WavePlayer_4, properties {soundWaveAssetPtr=/Game/Sounds/Asteria/Biomes/Eastern/Spots/Spot_Bird_Rare_Mountain_Crow_02}}
 {id=WavePlayer_5, properties {soundWaveAssetPtr=/Game/Sounds/Asteria/Biomes/Eastern/Spots/Spot_Bird_Rare_Mountain_Crow_03}}
 {id=WavePlayer_9, properties {soundWaveAssetPtr=/Game/Sounds/Helios/Biomes/BorealForest/Spots/Spot_Bird_Crow_DayEvening_Eastern_Europe_02}}
]
properties {
 attenuationSettings=&Fort_TileAmbient_OneShot_Bird_Att
 bEnableBusSends=false
 maxDistance=8005
 soundClassObject=&Fort_Ambient_OneShots_Outdoors
 volumeMultiplier=0.501
 preEffectBusSends [{}]
 soundSubmixSends [{maxSendLevel=1.2, minSendLevel=0.4, sendLevel=0, soundSubmix=/Engine/EngineSounds/Submixes/MasterReverbSubmixDefault}]
}
