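# Sound cue for a one-shot ambient spot effect at the CastleDoom POI: metallic, evil creature swells using indoor sound class and distant creak attenuation.
# A Random node picks evenly among 22 swell waves, feeding a Quality Level node and then a Modulator that varies pitch and volume.
# Only the first two quality levels connect to the waves; the other two are empty.
# A concurrency override limits simultaneous voices, and two distance-scaled sends feed default and distant reverb.
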SoundCue /Game/Sounds/Helios/POI/CastleDoom/Spots/Spot_Castle_Metallic_Evil_Creature_Swell_Cue
root=Modulator_0
nodes [
 {id=Modulator_0, children [QualityLevel_0], properties {pitchMin=0.8, volumeMax=1.2, volumeMin=0.85}}
 {id=QualityLevel_0, children [Random_0,Random_0,null,null]}
 {
  id=Random_0
  children [
   WavePlayer_0
   WavePlayer_1
   WavePlayer_2
   WavePlayer_3
   WavePlayer_4
   WavePlayer_5
   WavePlayer_6
   WavePlayer_7
   WavePlayer_8
   WavePlayer_9
   WavePlayer_10
   WavePlayer_11
   WavePlayer_12
   WavePlayer_13
   WavePlayer_14
   WavePlayer_32
   WavePlayer_33
   WavePlayer_34
   WavePlayer_35
   WavePlayer_36
   WavePlayer_37
   WavePlayer_38
  ]
  properties {weights [1,1,1,1,1,1,1,1,1,1,1,1,1,1,1,1,1,1,1,1,1,1]}
 }
 {id=WavePlayer_0, properties {soundWaveAssetPtr=&/Spot_Castle_Metallic_Evil_Creature_Swell_01}}
 {id=WavePlayer_1, properties {soundWaveAssetPtr=&/Spot_Castle_Metallic_Evil_Creature_Swell_02}}
 {id=WavePlayer_10, properties {soundWaveAssetPtr=&/Spot_Castle_Metallic_Evil_Creature_Swell_11}}
 {id=WavePlayer_11, properties {soundWaveAssetPtr=&/Spot_Castle_Metallic_Evil_Creature_Swell_12}}
 {id=WavePlayer_12, properties {soundWaveAssetPtr=&/Spot_Castle_Metallic_Evil_Creature_Swell_13}}
 {id=WavePlayer_13, properties {soundWaveAssetPtr=&/Spot_Castle_Metallic_Evil_Creature_Swell_14}}
 {id=WavePlayer_14, properties {soundWaveAssetPtr=&/Spot_Castle_Metallic_Evil_Creature_Swell_15}}
 {id=WavePlayer_2, properties {soundWaveAssetPtr=&/Spot_Castle_Metallic_Evil_Creature_Swell_03}}
 {id=WavePlayer_3, properties {soundWaveAssetPtr=&/Spot_Castle_Metallic_Evil_Creature_Swell_04}}
 {id=WavePlayer_32, properties {soundWaveAssetPtr=&/Spot_Castle_Metallic_Evil_Creature_Swell_16}}
 {id=WavePlayer_33, properties {soundWaveAssetPtr=&/Spot_Castle_Metallic_Evil_Creature_Swell_17}}
 {id=WavePlayer_34, properties {soundWaveAssetPtr=&/Spot_Castle_Metallic_Evil_Creature_Swell_18}}
 {id=WavePlayer_35, properties {soundWaveAssetPtr=&/Spot_Castle_Metallic_Evil_Creature_Swell_19}}
 {id=WavePlayer_36, properties {soundWaveAssetPtr=&/Spot_Castle_Metallic_Evil_Creature_Swell_20}}
 {id=WavePlayer_37, properties {soundWaveAssetPtr=&/Spot_Castle_Metallic_Evil_Creature_Swell_21}}
 {id=WavePlayer_38, properties {soundWaveAssetPtr=&/Spot_Castle_Metallic_Evil_Creature_Swell_22}}
 {id=WavePlayer_4, properties {soundWaveAssetPtr=&/Spot_Castle_Metallic_Evil_Creature_Swell_05}}
 {id=WavePlayer_5, properties {soundWaveAssetPtr=&/Spot_Castle_Metallic_Evil_Creature_Swell_06}}
 {id=WavePlayer_6, properties {soundWaveAssetPtr=&/Spot_Castle_Metallic_Evil_Creature_Swell_07}}
 {id=WavePlayer_7, properties {soundWaveAssetPtr=&/Spot_Castle_Metallic_Evil_Creature_Swell_08}}
 {id=WavePlayer_8, properties {soundWaveAssetPtr=&/Spot_Castle_Metallic_Evil_Creature_Swell_09}}
 {id=WavePlayer_9, properties {soundWaveAssetPtr=&/Spot_Castle_Metallic_Evil_Creature_Swell_10}}
]
properties {
 attenuationSettings=/Game/Sounds/Attenuation/Fort_TileAmbient_OneShot_Distant_Creak_Linear_Att
 bOverrideConcurrency=true
 maxDistance=8700
 soundClassObject=/Game/Sounds/FortSoundClassesAndModes/Fort_Ambient_OneShots_Indoors
 volumeMultiplier=0.501
 concurrencyOverrides {maxCount=2, voiceStealReleaseTime=2}
 soundSubmixSends [
  {
   maxSendDistance=8500
   minSendDistance=200
   minSendLevel=0.6
   soundSubmix=&Fortnite_Reverb_Default
   customSendLevelCurve {editorCurveData {keys [<at 0 0.25>,<at 1 1>]}}
  }
  {
   maxSendDistance=8500
   maxSendLevel=0.6
   minSendDistance=200
   minSendLevel=0.25
   soundSubmix=&Fortnite_Reverb_Distant
   customSendLevelCurve {editorCurveData {keys [
      <at 0 0>
      {arriveTangent=1, interpMode=RCIM_Cubic, leaveTangent=1, time=0.66, value=0.33}
      {arriveTangent=2.1868, interpMode=RCIM_Cubic, leaveTangent=2.1868, tangentMode=RCTM_User, time=1, value=1}
     ]}}
  }
 ]
}
