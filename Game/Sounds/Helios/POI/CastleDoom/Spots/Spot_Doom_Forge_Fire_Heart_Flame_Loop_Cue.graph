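# Looping ambient spot sound cue for the forge fire heart flame at the Castle Doom POI, built on a small torch flame loop.
# The looping wave passes through a pitch-fixed modulator into two nested, indefinitely looping envelopers.
# Their unequal cycles slowly dip pitch and volume separately, disguising repetition.
# It uses a shared small-fire attenuation asset, sends to reverb and overrides concurrency to cap simultaneous voices.
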SoundCue /Game/Sounds/Helios/POI/CastleDoom/Spots/Spot_Doom_Forge_Fire_Heart_Flame_Loop_Cue
root=Enveloper_3
nodes [
 {
  id=Enveloper_0
  children [Modulator_1]
  properties {
   bLoop=true
   bLoopIndefinitely=true
   loopEnd=30
   pitchCurve {editorCurveData {keys [
      {arriveTangent=-0.0013, interpMode=RCIM_Cubic, leaveTangent=-0.0013, time=0}
      <at 15 0.98 RCIM_Cubic>
      {arriveTangent=0.0013, interpMode=RCIM_Cubic, leaveTangent=0.0013, time=30}
     ]}}
   volumeCurve {editorCurveData {keys [<at 0 1>]}}
  }
 }
 {
  id=Enveloper_3
  children [Enveloper_0]
  properties {
   bLoop=true
   bLoopIndefinitely=true
   loopEnd=20
   pitchCurve {editorCurveData {keys [{time=0}]}}
   volumeCurve {editorCurveData {keys [
      {arriveTangent=-0.01, interpMode=RCIM_Cubic, leaveTangent=-0.01, time=0, value=1}
      <at 10 0.9 RCIM_Cubic>
      {arriveTangent=0.01, interpMode=RCIM_Cubic, leaveTangent=0.01, time=20, value=1}
     ]}}
  }
 }
 {id=Modulator_1, children [WavePlayer_4], properties {pitchMax=1, pitchMin=1}}
 {
  id=WavePlayer_4
  properties {bLooping=true, soundWaveAssetPtr=&/Spot_Fire_Small_Flame_Torch_Idle_Loop}
 }
]
properties {
 attenuationSettings=/Game/Sounds/Apollo/Biomes/UFO_Crash_Sites/FireLoops/Fire_Smallest_Close_MorePositional_Attenuation
 bBypassVolumeScaleForPriority=true
 bOverrideConcurrency=true
 maxDistance=860
 priority=0
 soundClassObject=/Game/Sounds/FortSoundClassesAndModes/Fort_Ambient_NonLoop
 volumeMultiplier=1
 attenuationOverrides {attenuationShape=Cone, bAttenuate=false, attenuationShapeExtents {x=171.3182}}
 concurrencyOverrides {maxCount=3, voiceStealReleaseTime=3}
 soundSubmixSends [{minSendLevel=0.6, soundSubmix=&Fortnite_Reverb_Default}]
}
