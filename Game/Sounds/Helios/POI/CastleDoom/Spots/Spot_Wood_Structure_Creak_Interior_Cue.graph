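# Sound cue for one-shot interior wood structure creaks, an indoor ambient spot sound in the CastleDoom POI.
# A Random node picks among fifteen equally weighted creak waves, feeding the first two Quality Level slots while the other two stay empty.
# A root Modulator varies pitch and volume.
# Distant-creak attenuation, a two-voice concurrency override and distance-scaled sends to default and distant reverb submixes shape playback.
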SoundCue /Game/Sounds/Helios/POI/CastleDoom/Spots/Spot_Wood_Structure_Creak_Interior_Cue
root=Modulator_0
nodes [
 {id=Modulator_0, children [QualityLevel_0], properties {pitchMax=1.1, pitchMin=0.8, volumeMin=0.85}}
 {id=QualityLevel_0, children [Random_1,Random_1,null,null]}
 {
  id=Random_1
  children [
   WavePlayer_8
   WavePlayer_9
   WavePlayer_10
   WavePlayer_11
   WavePlayer_12
   WavePlayer_13
   WavePlayer_14
   WavePlayer_15
   WavePlayer_16
   WavePlayer_17
   WavePlayer_18
   WavePlayer_19
   WavePlayer_20
   WavePlayer_21
   WavePlayer_22
  ]
  properties {weights [1,1,1,1,1,1,1,1,1,1,1,1,1,1,1]}
 }
 {id=WavePlayer_10, properties {soundWaveAssetPtr=&/Spot_Wood_Structure_Creak_Interior_03}}
 {id=WavePlayer_11, properties {soundWaveAssetPtr=&/Spot_Wood_Structure_Creak_Interior_04}}
 {id=WavePlayer_12, properties {soundWaveAssetPtr=&/Spot_Wood_Structure_Creak_Interior_05}}
 {id=WavePlayer_13, properties {soundWaveAssetPtr=&/Spot_Wood_Structure_Creak_Interior_06}}
 {id=WavePlayer_14, properties {soundWaveAssetPtr=&/Spot_Wood_Structure_Creak_Interior_07}}
 {id=WavePlayer_15, properties {soundWaveAssetPtr=&/Spot_Wood_Structure_Creak_Interior_08}}
 {id=WavePlayer_16, properties {soundWaveAssetPtr=&/Spot_Wood_Structure_Creak_Interior_09}}
 {id=WavePlayer_17, properties {soundWaveAssetPtr=&/Spot_Wood_Structure_Creak_Interior_10}}
 {id=WavePlayer_18, properties {soundWaveAssetPtr=&/Spot_Wood_Structure_Creak_Interior_11}}
 {id=WavePlayer_19, properties {soundWaveAssetPtr=&/Spot_Wood_Structure_Creak_Interior_12}}
 {id=WavePlayer_20, properties {soundWaveAssetPtr=&/Spot_Wood_Structure_Creak_Interior_13}}
 {id=WavePlayer_21, properties {soundWaveAssetPtr=&/Spot_Wood_Structure_Creak_Interior_14}}
 {id=WavePlayer_22, properties {soundWaveAssetPtr=&/Spot_Wood_Structure_Creak_Interior_15}}
 {id=WavePlayer_8, properties {soundWaveAssetPtr=&/Spot_Wood_Structure_Creak_Interior_01}}
 {id=WavePlayer_9, properties {soundWaveAssetPtr=&/Spot_Wood_Structure_Creak_Interior_02}}
]
properties {
 attenuationSettings=/Game/Sounds/Attenuation/Fort_TileAmbient_OneShot_Distant_Creak_Linear_Att
 bOverrideConcurrency=true
 maxDistance=8700
 soundClassObject=/Game/Sounds/FortSoundClassesAndModes/Fort_Ambient_OneShots_Indoors
 volumeMultiplier=0.501
 concurrencyOverrides {maxCount=2, voiceStealReleaseTime=2}
 soundSubmixSends [
  {
   maxSendDistance=8500
   minSendDistance=200
   minSendLevel=0.6
   soundSubmix=&Fortnite_Reverb_Default
   customSendLevelCurve {editorCurveData {keys [<at 0 0.25>,<at 1 1>]}}
  }
  {
   maxSendDistance=8500
   maxSendLevel=0.6
   minSendDistance=200
   minSendLevel=0.25
   soundSubmix=&Fortnite_Reverb_Distant
   customSendLevelCurve {editorCurveData {keys [
      <at 0 0>
      {arriveTangent=1, interpMode=RCIM_Cubic, leaveTangent=1, time=0.66, value=0.33}
      {arriveTangent=2.1868, interpMode=RCIM_Cubic, leaveTangent=2.1868, tangentMode=RCTM_User, time=1, value=1}
     ]}}
  }
 ]
}
